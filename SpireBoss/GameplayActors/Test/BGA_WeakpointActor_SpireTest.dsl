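; Test weakpoint actor for the Spire boss that activates its weakpoint as soon as play begins.
; On BeginPlay it runs the parent logic, fetches the weakpoint capsule, applies a collision kind to it, then enables the weakpoint.
; The construction script only calls its parent, so all other behavior comes from the inherited class.
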(event EventBeginPlay
  (|CallParentFunction)
  (CallFunction|SetWeakpointCollisionKind "NewEnumerator1" 0 (Variables|Default|GetWeakpointCapsule))
  (CallFunction|EnableWeakpoint))

(fn ConstructionScript
  (|CallParentFunction))
